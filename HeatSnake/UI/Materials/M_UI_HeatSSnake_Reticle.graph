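# Translucent UI material that draws the HeatSnake reticle procedurally, using FortUILibrary shape and reticle functions.
# It builds a center dot, two-sided crosshair, time-rotated segmented ring brackets, left/right SDF triangle arrows and four target-hit lines, each with an outline.
# Named reroutes combine these masks into tinted color and layered opacity under a master opacity.
# Scalar parameters control spread, size and per-element visibility, with arrows and hit lines hidden by default.
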Material /HeatSnake/UI/Materials/M_UI_HeatSSnake_Reticle
graph {
 properties {bAutomaticallySetUsageInEditor=false, blendMode=BLEND_Translucent, materialDomain=MD_UI, usedShadingModels=Default Lit}
 nodes [
  n1 = VectorParameter() {parameterName=Reticle Color, defaultValue=<rgba 1 1 1 1>}
  n2 = Multiply(A=NamedRerouteUsage(), B=ScalarParameter() {defaultValue=1, parameterName=Brackets Opacity})
  n3 = NamedRerouteUsage()
  n4 = Add(A=Add(A=Add(A=@n2, B=@n3), B=NamedRerouteUsage()), B=Multiply(A=NamedRerouteUsage(), B=ScalarParameter() {parameterName=Target Hit Opacity}))
  NamedRerouteDeclaration(None=Add(A=LinearInterpolate(B=@n1.RGB, Alpha=@n4), B=NamedRerouteUsage())) {name=Final Color, nodeColor=<rgba 0 0 0 1>}
  n5 = NamedRerouteUsage()
  n6 = Multiply(
   A=LinearInterpolate(A=Multiply(A=@n5) {constB=0.5}, Alpha=NamedRerouteUsage())
   B=ScalarParameter() {defaultValue=1, parameterName=Brackets Opacity}
  )
  n7 = LinearInterpolate(A=Multiply(A=NamedRerouteUsage()) {constB=0.5}, Alpha=NamedRerouteUsage())
  n8 = LinearInterpolate(A=Multiply(A=NamedRerouteUsage()) {constB=0.5}, Alpha=NamedRerouteUsage())
  n9 = ScalarParameter() {parameterName=Arrows Opacity}
  n10 = Multiply(A=NamedRerouteUsage()) {constB=0.5}
  n11 = NamedRerouteUsage()
  n12 = Add(
   A=Add(A=Add(A=@n6, B=@n7), B=Multiply(A=@n8, B=@n9))
   B=Multiply(A=LinearInterpolate(A=@n10, Alpha=@n11), B=ScalarParameter() {parameterName=Target Hit Opacity})
  )
  n13 = NamedRerouteUsage()
  n14 = LinearInterpolate(A=Multiply(A=Constant() {r=0.5}, B=NamedRerouteUsage()), B=@n13, Alpha=@n13)
  NamedRerouteDeclaration(None=Multiply(A=Add(A=@n12, B=@n14), B=ScalarParameter() {defaultValue=0.6, parameterName=Master Opacity})) {name=Opacity, nodeColor=<rgba 0 1 0.3672 1>}
  n15 = MaterialFunctionCall(
   Size=ScalarParameter() {defaultValue=0.025, parameterName=Center Dot Scale}
   OutlineThickness=ScalarParameter() {defaultValue=0.015, parameterName=Center Dot Outline Thickness}
  ) {materialFunction=&MF_UI_Circle}
  NamedRerouteDeclaration(None=@n15.Fill) {name=Center Dot, nodeColor=<rgba 0.7578 1 0 1>}
  NamedRerouteDeclaration(None=Power(Base=@n15.Outline) {constExponent=0.01}) {name=Center dot outline, nodeColor=<rgba 0.9219 0 1 1>}
  n16 = ScalarParameter() {defaultValue=90, parameterName=Cross Hair Spread}
  n17 = MaterialFunctionCall(
   UVs=MaterialFunctionCall!(
    Reticle UVs
    Reticle Sides=Constant() {r=2}
    Reticle Scale=Constant() {r=0.15}
    Reticle Rotation=Constant() {r=0.25}
    Prong Scale
    Prong Rotation
    Spread Offset=MaterialFunctionCall!(Positional Spread Offset=@n16, Accuracy Spread Offset, Min Value, "Input Range [Min,Max]", "Remapped Range [Min,Max]") {materialFunction=/FortUILibrary/Systems/Reticles/MaterialFunctions/MF_UI_Reticle_Remap_Spread}.Result
    Positional Offset
    "Spread Input Range [Min,Max]"
    "Spread Remapped Range [Min,Max]"
   ) {materialFunction=/FortUILibrary/Systems/Reticles/MaterialFunctions/MF_UI_Reticle_UVs}.UVs
   Size=AppendVector(A=Constant() {r=0.2}, B=ScalarParameter() {defaultValue=1, parameterName=Crosshair Size})
   OutlineThickness=Constant() {r=0.2}
  ) {materialFunction=&MF_UI_Box}
  NamedRerouteDeclaration(None=Multiply(A=@n17.Outline, B=ScalarParameter() {defaultValue=1, parameterName=Cross Hair Opacity})) {name=CrossHairOutline, nodeColor=<rgba 0 0 0 1>}
  n18 = MaterialFunctionCall(
   UVs=TextureCoordinate()
   Rotation Center=Constant() {r=0.5}
   "Rotation Angle (0-1)"=Multiply(A=Time(), B=ScalarParameter() {defaultValue=-0.3, parameterName=Rotation Speed}) {constB=-0.3}
  ) {materialFunction=&CustomRotator}
  n19 = SmoothStep(Value=MaterialFunctionCall(UVs=@n18.Rotated Values, Size=ScalarParameter() {defaultValue=0.35, parameterName=Brackets Size}, OutlineThickness=Constant() {r=0.01}) {materialFunction=&MF_UI_Circle}.SDFOutline) {constMax=0.01}
  n20 = SmoothStep(Value=MaterialFunctionCall!(UV=@n18.Rotated Values, Segment Count=Constant() {r=4}, Segment Spacing=Constant() {r=0.5}, Fixed Spacing) {materialFunction=/FortUILibrary/MaterialFunctions/Basics/MF_UI_RadialSegments}.Segmented Triangles Gradient) {constMax=0.03}
  NamedRerouteDeclaration(None=Multiply(A=OneMinus(None=@n19), B=OneMinus(None=@n20))) {name=RoundBrackets, nodeColor=<rgba 0 0 0 1>}
  n21 = SmoothStep(Value=MaterialFunctionCall(UVs=@n18.Rotated Values, Size=ScalarParameter() {defaultValue=0.35, parameterName=Brackets Size}, OutlineThickness=Constant() {r=0.03}) {materialFunction=&MF_UI_Circle}.SDFOutline) {constMax=0.01}
  n22 = SmoothStep(Value=MaterialFunctionCall(UV=@n18.Rotated Values, Segment Count=Constant() {r=4}, Segment Spacing=Constant() {r=0.55}) {materialFunction=&MF_UI_RadialSegments}.Segmented Triangles Gradient) {constMax=0.03}
  NamedRerouteDeclaration(None=Multiply(A=OneMinus(None=@n21), B=OneMinus(None=@n22))) {name=RoundBracketsOutlines, nodeColor=<rgba 0 0 0 1>}
  n23 = Add(A=TextureCoordinate(), B=AppendVector(A=Constant(), B=ScalarParameter() {defaultValue=-0.3, parameterName=Right Arrow Spread}))
  n24 = MaterialFunctionCall!(
   UVs=MaterialFunctionCall(
    UVs=MaterialFunctionCall(UVs=@n23) {materialFunction=&CustomRotator}.Rotated Values
    Texture Scale=VectorParameter() {parameterName=TriangleScale, defaultValue=<rgba 1.1548 0.6564 1.2 1>}.RGB
   ) {materialFunction=&ScaleUVsByCenter}.UVs
   Radius=ScalarParameter() {defaultValue=0.06, parameterName=Arrow Size}
   Sides=Constant() {r=3}
   InnerAngle
  ) {materialFunction=/FortUILibrary/MaterialFunctions/Shapes/SDFs/MF_SDF_RegularPolygon}
  n25 = NamedRerouteDeclaration(None=MaterialFunctionCall(DistanceField=@n24.SDF) {materialFunction=&MF_UI_AntiAliasing}.Inner) {name=ArrowRight, nodeColor=<rgba 0 0 0 1>}
  n26 = Add(A=TextureCoordinate(), B=AppendVector(A=Constant(), B=ScalarParameter() {defaultValue=0.3, parameterName=Left Arrow Spread}))
  n27 = Constant() {r=-0.5}
  n28 = MaterialFunctionCall(
   UVs=MaterialFunctionCall(
    UVs=MaterialFunctionCall(UVs=@n26, "Rotation Angle (0-1)"=@n27) {materialFunction=&CustomRotator}.Rotated Values
    Texture Scale=VectorParameter() {parameterName=TriangleScale, defaultValue=<rgba 1.1548 0.6564 1.2 1>}.RGB
   ) {materialFunction=&ScaleUVsByCenter}.UVs
   Radius=ScalarParameter() {defaultValue=0.06, parameterName=Arrow Size}
   Sides=Constant() {r=3}
  ) {materialFunction=&MF_SDF_RegularPolygon}
  n29 = NamedRerouteDeclaration(None=MaterialFunctionCall(DistanceField=@n28.SDF) {materialFunction=&MF_UI_AntiAliasing}.Inner) {name=ArrowLeft, nodeColor=<rgba 0 0 0 1>}
  NamedRerouteDeclaration(None=Add(A=@n25, B=@n29)) {name=Arrows, nodeColor=<rgba 0 0 0 1>}
  n30 = Subtract(
   A=MaterialFunctionCall!(
    DistanceField=@n24.SDF
    Threshold
    Thickness=ScalarParameter() {defaultValue=0.0185, parameterName=Arrow Outline Thickness}
    Alignment
    Smoothing
   ) {materialFunction=/FortUILibrary/MaterialFunctions/Shapes/Modifiers/MF_UI_ShapeOutline}.Outline
   B=@n25
  )
  n31 = Subtract(
   A=MaterialFunctionCall(DistanceField=@n28.SDF, Thickness=ScalarParameter() {defaultValue=0.0185, parameterName=Arrow Outline Thickness}) {materialFunction=&MF_UI_ShapeOutline}.Outline
   B=@n29
  )
  NamedRerouteDeclaration(None=Add(
    A=NamedRerouteDeclaration(None=@n30) {name=ArrowRightOutline, nodeColor=<rgba 0 0 0 1>}
    B=NamedRerouteDeclaration(None=@n31) {name=ArrowLeftOutline, nodeColor=<rgba 0 0 0 1>}
   )) {name=ArrowsOutlines, nodeColor=<rgba 0 0 0 1>}
  n32 = ScalarParameter() {defaultValue=150, parameterName=SpreadTargetHit}
  n33 = MaterialFunctionCall(
   UVs=MaterialFunctionCall(
    Reticle Sides=Constant() {r=4}
    Reticle Scale=Constant() {r=0.15}
    Reticle Rotation=Constant() {r=0.25}
    Spread Offset=MaterialFunctionCall(Positional Spread Offset=@n32) {materialFunction=&MF_UI_Reticle_Remap_Spread}.Result
   ) {materialFunction=&MF_UI_Reticle_UVs}.UVs
   Size=Constant2Vector() {g=2, r=0.2}
   OutlineThickness=Constant() {r=0.2}
  ) {materialFunction=&MF_UI_Box}
  NamedRerouteDeclaration(None=@n33.Fill) {name=TargetHitLines, nodeColor=<rgba 0 0 0 1>}
  NamedRerouteDeclaration(None=@n33.Outline) {name=TargetHitLinesOutline, nodeColor=<rgba 0 0 0 1>}
  NamedRerouteDeclaration(None=Multiply(A=@n17.Fill, B=ScalarParameter() {defaultValue=1, parameterName=Cross Hair Opacity})) {name=CrossHair, nodeColor=<rgba 0 0 0 1>}
  NamedRerouteUsage()
  NamedRerouteUsage()
 ]
}
