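# HeatSnake grapple auto-reticle widget that shows when the weapon's targeting component reports a target.
# Target-type changes drive visibility and in-range animations, while hookshot targeting plays a crosshair-to-brackets transition forward or in reverse.
# Async gameplay-tag waits on the owning pawn hide the reticle during movement cooldown and restore it afterwards.
# An enemy hit animation handler exists, but its impact-delegate binding is not wired into execution.
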WidgetBlueprint /HeatSnake/UI/WBP_HeatSnakeGrappleAutoreticle
animations [
 EnterHookshot {
  binding_count=1
  length_seconds=0.6
  object=&self:EnterHookshot
  sequence {
   bindings [Reticle {
     id=#1
     possessed_class=Image
     tracks [
      {
       type=MovieScene2DTransformTrack
       sections [{
         type=MovieScene2DTransformSection
         end_frame=12
         start_frame=0
         channels [
          Translation.X {type=MovieSceneScriptingFloatChannel, const [0,1]}
          Translation.Y {type=MovieSceneScriptingFloatChannel, const [0,1]}
          Angle {type=MovieSceneScriptingFloatChannel, keys [[0,0],[8.5715,90]]}
          Scale.X {type=MovieSceneScriptingFloatChannel, const [1,1]}
          Scale.Y {type=MovieSceneScriptingFloatChannel, const [1,1]}
          Shear.X {type=MovieSceneScriptingFloatChannel, const [0,1]}
          Shear.Y {type=MovieSceneScriptingFloatChannel, const [0,1]}
         ]
        }]
      }
      {
       type=MovieSceneWidgetMaterialTrack
       sections [{
         type=MovieSceneParameterSection
         end_frame=18
         start_frame=0
         channels [
          Brackets Opacity {type=MovieSceneScriptingFloatChannel, keys [[0,0],[8.5715,1]]}
          Cross Hair Opacity {type=MovieSceneScriptingFloatChannel, keys [[0,1],[8.5715,0]]}
         ]
        }]
      }
     ]
    }]
   display_rate {numerator=30}
  }
 }
 EnemyInRange {
  binding_count=1
  length_seconds=0.35
  object=&self:EnemyInRange
  sequence {
   bindings [Reticle {
     id=#2
     possessed_class=Image
     tracks [{
       type=MovieSceneWidgetMaterialTrack
       sections [{
         type=MovieSceneParameterSection
         end_frame=10
         start_frame=0
         channels [
          Reticle ColorR {type=MovieSceneScriptingFloatChannel, const [1,2]}
          Reticle ColorG {type=MovieSceneScriptingFloatChannel, keys [[0,1],[4,0]]}
          Reticle ColorB {type=MovieSceneScriptingFloatChannel, keys [[0,1],[4,0]]}
          Reticle ColorA {type=MovieSceneScriptingFloatChannel, const [1,1]}
         ]
        }]
      }]
    }]
   display_rate {numerator=30}
  }
 }
 ActorInRange {
  binding_count=0
  length_seconds=0.35
  object=&self:ActorInRange
  sequence {bindings [], display_rate {numerator=30}}
 }
 AnythingInRange {
  binding_count=1
  length_seconds=0.45
  object=&self:AnythingInRange
  sequence {
   bindings [Reticle {
     id=#3
     possessed_class=Image
     tracks [{
       type=MovieSceneWidgetMaterialTrack
       sections [{
         type=MovieSceneParameterSection
         end_frame=11
         start_frame=-2
         channels [Master Opacity {type=MovieSceneScriptingFloatChannel, keys [[-1,0],[4,0.6]]}]
        }]
      }]
    }]
   display_rate {numerator=30}
  }
 }
 HitAnim {
  binding_count=2
  length_seconds=0.1667
  object=&self:HitAnim
  sequence {
   bindings [
    Reticle {
     id=#4
     possessed_class=Image
     tracks [{
       type=MovieSceneWidgetMaterialTrack
       sections [{
         type=MovieSceneParameterSection
         end_frame=5
         start_frame=0
         channels [
          Target Hit Opacity {type=MovieSceneScriptingFloatChannel, keys [[0,0],[2,0.5],[4.999,0]]}
          SpreadTargetHit {type=MovieSceneScriptingFloatChannel, keys [[0,120],[4.999,200]]}
         ]
        }]
      }]
    }
    WBP_HeatSnake_Reticle_C_37 {
     id=#5
     possessed_class=WBP_HeatSnake_BaseReticle_C
     tracks [{
       type=MovieScene2DTransformTrack
       sections [{
         type=MovieScene2DTransformSection
         end_frame=5
         start_frame=0
         channels [
          Translation.X {type=MovieSceneScriptingFloatChannel, const [0,1]}
          Translation.Y {type=MovieSceneScriptingFloatChannel, const [0,1]}
          Angle {type=MovieSceneScriptingFloatChannel, keys [[0,0],[0.7565,-5],[1.6665,3],[2.754,-2],[3.939,1],[5,0]]}
          Scale.X {type=MovieSceneScriptingFloatChannel, keys [[0,1],[1.6665,0.85],[5,1]]}
          Scale.Y {type=MovieSceneScriptingFloatChannel, keys [[0,1],[1.6665,0.85],[5,1]]}
          Shear.X {type=MovieSceneScriptingFloatChannel, const [0,1]}
          Shear.Y {type=MovieSceneScriptingFloatChannel, const [0,1]}
         ]
        }]
      }]
    }
   ]
   display_rate {numerator=30}
  }
 }
]
graphs [
 EventGraph {nodes [
   n1 = TargetingTargetUpdated!() -> (OutputDelegate, then, TargetType:HeatSnakeGrappleTargetTypeEnum Enum=NewEnumerator0) {is=title K2Node_CustomEvent}
   > Branch(Condition="Not Equal (Enum)"!(
     A:HeatSnakeGrappleTargetTypeEnum Enum=@n1.TargetType
     B:HeatSnakeGrappleTargetTypeEnum Enum=Get LastTargetType() -> (LastTargetType:HeatSnakeGrappleTargetTypeEnum Enum=NewEnumerator0)
    ) -> (ReturnValue:bool))
   > n2 = Set LastTargetType(LastTargetType:HeatSnakeGrappleTargetTypeEnum Enum=@n1.TargetType) -> (Output_Get:HeatSnakeGrappleTargetTypeEnum Enum=NewEnumerator0)
   > StopAnyInRangeAnims()
   > n3 = Switch on HeatSnakeGrappleTargetTypeEnum!(execute, Selection:HeatSnakeGrappleTargetTypeEnum Enum=@n2) -> (NewEnumerator0:Exec, NewEnumerator1:Exec) {is=title K2Node_SwitchEnum}
   > PlayActorInRange!(execute, self, IsActorInRange:bool=false) -> (then)
   > PlayEnemyInRange!(execute, self, IsEnemyInRange?:bool=false) -> (then)
   > PlayAnythingInRange!(execute, self, InRange:bool=false) -> (then)
   > Set Visibility(Visibility=Collapsed, self=Get Reticle() -> (Reticle:Image Object Reference))
   PlayActorInRange(execute=@n3.NewEnumerator1, IsActorInRange=true)
   > Set IsEnemyInRange?(IsEnemyInRange?:bool=false) -> (Output_Get:bool)
   > PlayAnythingInRange(InRange=true)
   > Branch(Condition=Get ReticleEnabledByTag() -> (ReticleEnabledByTag:bool))
   > Set Visibility(Visibility=Visible, self=Get Reticle())
   n4 = HandleTargetingHookshotUpdated!() -> (OutputDelegate, then, bTargetingHookshot:bool=~) {is=title K2Node_CustomEvent}
   > n5 = Branch(Condition=@n4.bTargetingHookshot)
   n6 = Get EnterHookshot() -> (EnterHookshot:Widget Animation Object Reference)
   > PlayAnimationForward(InAnimation=@n6)
   PlayAnimationReverse(execute=@n5.else, InAnimation=@n6)
   n7 = OnPlayImpactFXDelegate_Event!() -> (
    OutputDelegate
    then
    Weapon:Fort Weapon Object Reference
    HitResult:"Hit Result Structure (by ref)"
    FXSystemComponent:FXSystem Component Object Reference
   ) {is=title K2Node_CustomEvent}
   > Get!(execute:Exec) -> (then:Exec, else:Exec, IsEnemyInRange?:bool)
   n8 = Get HitAnim() -> (HitAnim:Widget Animation Object Reference)
   > StopAnimation(InAnimation=@n8)
   > PlayAnimationForward(InAnimation=@n8, PlaybackSpeed=0.45, bRestoreState=true)
   Event Construct()
   > n9 = Cast To FortPlayerPawnAthena(Object=Get Controlled Pawn!(self:Controller Object Reference=GetOwningPlayer()) -> (ReturnValue:Pawn Object Reference) {is=title K2Node_CallFunction})
   > Set As Fort Player Pawn Athena(As Fort Player Pawn Athena:Fort Player Pawn Athena Object Reference=@n9) -> (Output_Get:Fort Player Pawn Athena Object Reference)
   > n10 = WaitGameplayTagAddToActor!(
    execute:Exec
    TargetActor:Actor Object Reference=Get As Fort Player Pawn Athena() -> (As Fort Player Pawn Athena:Fort Player Pawn Athena Object Reference)
    Tag:Gameplay Tag Structure=Get CooldownListeningTag() -> (CooldownListeningTag:Gameplay Tag Structure)
    OnlyTriggerOnce:bool=false
   ) -> (then:Exec, AsyncTaskProxy:Ability Async Wait Gameplay Tag Added Object Reference, Added:Exec) {is=title K2Node_AsyncAction}
   > Set CooldownAddTask(CooldownAddTask:Ability Async Wait Gameplay Tag Added Object Reference=@n10) -> (Output_Get:Ability Async Wait Gameplay Tag Added Object Reference)
   > n11 = WaitGameplayTagRemoveFromActor!(
    execute:Exec
    TargetActor:Actor Object Reference=Get As Fort Player Pawn Athena()
    Tag:Gameplay Tag Structure=Get CooldownListeningTag()
    OnlyTriggerOnce:bool=false
   ) -> (then:Exec, AsyncTaskProxy:Ability Async Wait Gameplay Tag Removed Object Reference, Removed:Exec) {is=title K2Node_AsyncAction}
   > Set CooldownRemoveTask(CooldownRemoveTask:Ability Async Wait Gameplay Tag Removed Object Reference=@n11) -> (Output_Get:Ability Async Wait Gameplay Tag Removed Object Reference)
   n12 = GetEquippedWeapon!(self) -> (Out:Fort Weapon Object Reference)
   > Is Valid(InputObject=@n12)
   > n13 = Set HeatSnakeComponent(HeatSnakeComponent:B Heat Snake Grapple Targeting Component Object Reference=GetComponentByClass!(
     self:Actor Object Reference=@n12
     ComponentClass:Actor Component Class Reference=/HeatSnake/Gameplay/Targeting/B_HeatSnakeGrappleTargetingComponent.B_HeatSnakeGrappleTargetingComponent_C
    ) -> (ReturnValue:B Heat Snake Grapple Targeting Component Object Reference)) -> (Output_Get:B Heat Snake Grapple Targeting Component Object Reference)
   > Is Valid(InputObject=@n13)
   n14 = Create Event()
   > Bind Event to On Targeting Target Updated(self:B Heat Snake Grapple Targeting Component Object Reference=@n13, Delegate=@n14) {is=title K2Node_AddDelegate}
   n15 = Create Event()
   > Bind Event to On Targeting Hookshot(self:B Heat Snake Grapple Targeting Component Object Reference=@n13, Delegate=@n15) {is=title K2Node_AddDelegate}
   Set ReticleEnabledByTag(execute=@n11.Removed, ReticleEnabledByTag:bool=true) -> (Output_Get:bool)
   n16 = Get LastTargetType() -> (LastTargetType=NewEnumerator0)
   > Branch(Condition=NOT Boolean(A="Equal (Enum)"(A:HeatSnakeGrappleTargetTypeEnum Enum=@n16, B:HeatSnakeGrappleTargetTypeEnum Enum=NewEnumerator0)))
   > Set Visibility(Visibility=Visible, self=Get Reticle())
   Set ReticleEnabledByTag(execute=@n10.Added, ReticleEnabledByTag=false)
   > Set Visibility(Visibility=Collapsed, self=Get Reticle())
   PrintString!(execute=@n9.CastFailed, InString=HeatSnake UI: Cannot Resolve Owner Pawn, bPrintToScreen, bPrintToLog, TextColor, Duration, Key) -> (then)
   Event Destruct()
   > EndAction(self:Ability Async Object Reference=Get CooldownAddTask() -> (CooldownAddTask:Ability Async Wait Gameplay Tag Added Object Reference))
   > EndAction(self=Get CooldownRemoveTask() -> (CooldownRemoveTask:Ability Async Wait Gameplay Tag Removed Object Reference))
   > Is Valid(InputObject=Get HeatSnakeComponent() -> (HeatSnakeComponent:B Heat Snake Grapple Targeting Component Object Reference))
   > Unbind Event from On Targeting Target Updated(self:B Heat Snake Grapple Targeting Component Object Reference=@n13, Delegate=@n14) {is=title K2Node_RemoveDelegate}
   > Unbind Event from On Targeting Hookshot(self:B Heat Snake Grapple Targeting Component Object Reference=@n13, Delegate=@n15) {is=title K2Node_RemoveDelegate}
   Assign On Is Valid Target Changed!(
    execute:Exec
    self:Fort Weapon Component Target State Object Reference=GetComponentByClass(ComponentClass=/Script/FortniteGame.FortWeaponComponent_TargetState) -> (ReturnValue:Fort Weapon Component Target State Object Reference)
    Delegate:"Delegate (by ref)"=OnIsValidTargetChanged_Event!() -> (OutputDelegate, then, bIsValidTarget:bool=~) {is=title K2Node_CustomEvent}
   ) -> (then:Exec) {is=title K2Node_AssignDelegate}
   Assign On Play Impact FXDelegate!(execute:Exec, self:Fort Weapon Object Reference=GetEquippedWeapon(), Delegate:"Delegate (by ref)"=@n7) -> (then:Exec) {is=title K2Node_AssignDelegate}
  ]}
 GetEquippedWeapon {nodes [
   GetEquippedWeapon() {is=title K2Node_FunctionEntry}
   > n1 = Cast To FortPlayerPawn(Object=Get Controlled Pawn(self=GetOwningPlayer()))
   > Return Node!(
    execute
    Out:Fort Weapon Object Reference=Get CurrentWeapon!(self:Fort Pawn Object Reference=@n1) -> (CurrentWeapon:Fort Weapon Object Reference)
   )
  ]}
 PlayActorInRange {nodes [
   n1 = PlayActorInRange!() -> (then, IsActorInRange:bool=~) {is=title K2Node_FunctionEntry}
   > n2 = Delta Gate Wildcard!(execute:Exec, Variable:"Boolean (by ref)"=Get IsActorInRange?() -> (IsActorInRange?:bool), New Value:bool=@n1) -> (Success:Exec, Updated Variable:"Boolean (by ref)", Previous Variable:bool=~) {is=title K2Node_MacroInstance}
   > n3 = Branch(Condition=@n2)
   n4 = Get ActorInRange() -> (ActorInRange:Widget Animation Object Reference)
   > PlayAnimationForward(InAnimation=@n4)
   PlayAnimationReverse(execute=@n3.else, InAnimation=@n4)
  ]}
 PlayAnythingInRange {nodes [
   n1 = PlayAnythingInRange!() -> (then, InRange:bool=~) {is=title K2Node_FunctionEntry}
   > n2 = Branch(Condition=@n1)
   n3 = Get AnythingInRange() -> (AnythingInRange:Widget Animation Object Reference)
   > PlayAnimationForward(InAnimation=@n3)
   PlayAnimationReverse(execute=@n2.else, InAnimation=@n3)
  ]}
 PlayEnemyInRange {nodes [
   n1 = PlayEnemyInRange!() -> (then, IsEnemyInRange?:bool=~) {is=title K2Node_FunctionEntry}
   > n2 = Delta Gate Wildcard(Variable=Get IsEnemyInRange?() -> (IsEnemyInRange?:bool), New Value=@n1) -> (Previous Variable=~)
   > n3 = Branch(Condition=@n2)
   n4 = Get EnemyInRange() -> (EnemyInRange:Widget Animation Object Reference)
   > PlayAnimationForward(InAnimation=@n4)
   PlayAnimationReverse(execute=@n3.else, InAnimation=@n4)
  ]}
 StopAnyInRangeAnims {nodes [
   StopAnyInRangeAnims() {is=title K2Node_FunctionEntry}
   > StopAnimation(InAnimation=Get ActorInRange())
   > StopAnimation(InAnimation=Get EnemyInRange())
   > StopAnimation(InAnimation=Get AnythingInRange())
  ]}
]
variables [
 As Fort Player Pawn Athena {type=FortPlayerPawnAthena}
 CooldownAddTask {type=AbilityAsync_WaitGameplayTagAdded, default=null}
 CooldownListeningTag {type=GameplayTag, default {tagName=Granted.Cooldown.HeatSnake.Movement}}
 CooldownRemoveTask {type=AbilityAsync_WaitGameplayTagRemoved, default=null}
 HeatSnakeComponent {type=B_HeatSnakeGrappleTargetingComponent_C, default=null}
 IsActorInRange? {type=boolean}
 IsEnemyInRange? {type=boolean}
 LastTargetType {type=HeatSnakeGrappleTargetTypeEnum, default=None}
 ReticleEnabledByTag {type=boolean}
]
widgets {
 info {inheritedWidgetCount=1, parentClass=/Script/FortniteUI.FortWeaponAttachmentPointWidget, rootWidgetClass=&Overlay, widgetCount=4}
 widgets [
  Overlay_21 {widgetClassPath=&Overlay, widget=Overlay_21, properties {clipping=ClipToBounds}}
  Dots {
   widgetClassPath=&Image
   bIsVariable=true
   parent=Overlay_21
   widget=Dots
   properties {
    visibility=Hidden
    brush {
     bIsDynamicallyLoaded=false
     drawAs=Image
     imageType=NoImage
     mirroring=NoMirror
     resourceName=None
     resourceObject=/SunRoseChainWhipGameplay/Items/ChainWhip/UI/MI_Ret_ChainPips
     tiling=NoTile
     imageSize=<350 350>
     margin=<ltrb 0 0 0 0>
     outlineSettings {
      bUseBrushTransparency=false
      roundingType=HalfHeightRadius
      width=0
      color {colorUseRule=UseColor_Specified, specifiedColor=<rgba 0 0 0 0>}
      cornerRadii=<0 0 0 0>
     }
     tintColor {colorUseRule=UseColor_Specified, specifiedColor=<rgba 1 1 1 1>}
     uVRegion {bIsValid=false, max=<0 0>, min=<0 0>}
    }
   }
   slot {class=OverlaySlot, properties {horizontalAlignment=HAlign_Center, verticalAlignment=VAlign_Center}}
  }
  Reticle {
   widgetClassPath=&Image
   bIsVariable=true
   parent=Overlay_21
   widget=Reticle
   properties {brush {
     bIsDynamicallyLoaded=false
     drawAs=Image
     imageType=FullColor
     mirroring=NoMirror
     resourceName=None
     resourceObject=/HeatSnake/UI/Materials/MI_UI_HeatSSnake_Reticle_Grapple
     tiling=NoTile
     imageSize=<150 150>
     margin=<ltrb 0 0 0 0>
     outlineSettings {
      bUseBrushTransparency=false
      roundingType=HalfHeightRadius
      width=0
      color {colorUseRule=UseColor_Specified, specifiedColor=<rgba 0 0 0 0>}
      cornerRadii=<0 0 0 0>
     }
     tintColor {colorUseRule=UseColor_Specified, specifiedColor=<rgba 1 1 1 1>}
     uVRegion {bIsValid=false, max=<0 0>, min=<0 0>}
    }}
   slot {class=OverlaySlot, properties {horizontalAlignment=HAlign_Center, verticalAlignment=VAlign_Center}}
  }
  TutorialHighlight {widgetClassPath=&TutorialHighlight}
 ]
}
