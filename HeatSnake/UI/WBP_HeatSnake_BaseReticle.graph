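# HeatSnake weapon reticle extension widget that shows the movement cooldown as a center-filling progress bar over the reticle image.
# Construct broadcasts an auto-aim show event and starts async cooldown-tag listeners on the owning pawn; Destruct broadcasts hide.
# Tag add/remove toggles the bar and a looping timer that polls the cooldown GameplayEffect.
# Percent is remaining/total duration, passed through MapRangeClamped with total duration as input maximum.
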WidgetBlueprint /HeatSnake/UI/WBP_HeatSnake_BaseReticle
graphs [EventGraph {nodes [
   n1 = CustomEvent() {is=title K2Node_CustomEvent}
   > n2 = Get Active Gameplay Effects for Query!(
    execute
    self:Ability System Component Object Reference=GetFortAbilitySystemComponent!(Actor:Actor Object Reference=Get As Fort Player Pawn Athena() -> (As Fort Player Pawn Athena:Fort Player Pawn Athena Object Reference)) -> (ReturnValue:Fort Ability System Component Object Reference)
    Query:"Gameplay Effect Query Structure (by ref)"=Make Gameplay Effect Query!(
     CustomMatchDelegate_BP:Delegate="(null).None"
     OwningTagQuery:Gameplay Tag Query Structure=(TokenStreamVersion=0,TagDictionary=,QueryTokenStream=,UserDescription="",AutoDescription="")
     EffectTagQuery:Gameplay Tag Query Structure=(TokenStreamVersion=0,TagDictionary=,QueryTokenStream=,UserDescription="",AutoDescription="")
     SourceTagQuery:Gameplay Tag Query Structure=(TokenStreamVersion=0,TagDictionary=,QueryTokenStream=,UserDescription="",AutoDescription="")
     SourceAggregateTagQuery:Gameplay Tag Query Structure=(TokenStreamVersion=0,TagDictionary=,QueryTokenStream=,UserDescription="",AutoDescription="")
     ModifyingAttribute:Gameplay Attribute Structure=(AttributeName="",Attribute=,AttributeOwner=None)
     EffectSource:Object Reference
     EffectDefinition:Gameplay Effect Class Reference=/HeatSnake/Gameplay/Effects/GE_HeatSnakeMovementCooldown.GE_HeatSnakeMovementCooldown_C
    ) -> (GameplayEffectQuery:Gameplay Effect Query Structure) {is=title K2Node_MakeStruct}
   ) -> (then, ReturnValue:Array of Active Gameplay Effect Handle Structures) {is=title K2Node_CallFunction}
   n3 = "Get (a ref)"(Array:Array of Active Gameplay Effect Handle Structures=@n2, Dimension 1=0) -> (Output:"Active Gameplay Effect Handle Structure (by ref)")
   > n4 = GetActiveGameplayEffectTotalDuration!(execute, ActiveHandle:Active Gameplay Effect Handle Structure=@n3) -> (then, ReturnValue:float)
   > n5 = GetActiveGameplayEffectRemainingDuration!(execute, ActiveHandle:Active Gameplay Effect Handle Structure=@n3) -> (then, ReturnValue:float)
   > Set Percent!(
    execute
    Percent:float=MapRangeClamped(Value=float / float(A=@n5, B=@n4), InRangeA=0, InRangeB=@n4)
    self:Progress Bar Object Reference=Get ProgressBar_130() -> (ProgressBar_130:Progress Bar Object Reference)
   ) -> (then, Output_Get:float)
   Event Construct()
   > Broadcast Event!(
    execute
    self:Gameplay Event Router Component Object Reference=GetGlobalEventRouter()
    EventData:ST Heat Snake Auto Aim Event Data Structure=Make ST Heat Snake Auto Aim Event Data!(ShouldShow_2_#1:bool=true) -> (ST_HeatSnake_AutoAimEventData:ST Heat Snake Auto Aim Event Data Structure) {is=title K2Node_MakeStruct}
    EventContext:Object Reference
   ) -> (then, ReturnValue:bool)
   > n6 = Cast To FortPlayerPawnAthena(Object=Get Controlled Pawn!(self:Controller Object Reference=GetOwningPlayer()) -> (ReturnValue:Pawn Object Reference) {is=title K2Node_CallFunction})
   > Set As Fort Player Pawn Athena(As Fort Player Pawn Athena:Fort Player Pawn Athena Object Reference=@n6) -> (Output_Get:Fort Player Pawn Athena Object Reference)
   > n7 = WaitGameplayTagAddToActor!(
    execute:Exec
    TargetActor:Actor Object Reference=Get As Fort Player Pawn Athena()
    Tag:Gameplay Tag Structure=Get CooldownListeningTag() -> (CooldownListeningTag:Gameplay Tag Structure)
    OnlyTriggerOnce:bool=false
   ) -> (then:Exec, AsyncTaskProxy:Ability Async Wait Gameplay Tag Added Object Reference, Added:Exec) {is=title K2Node_AsyncAction}
   > Set CooldownAddTask(CooldownAddTask:Ability Async Wait Gameplay Tag Added Object Reference=@n7) -> (Output_Get:Ability Async Wait Gameplay Tag Added Object Reference)
   > n8 = WaitGameplayTagRemoveFromActor!(
    execute:Exec
    TargetActor:Actor Object Reference=Get As Fort Player Pawn Athena()
    Tag:Gameplay Tag Structure=Get CooldownListeningTag()
    OnlyTriggerOnce:bool=false
   ) -> (then:Exec, AsyncTaskProxy:Ability Async Wait Gameplay Tag Removed Object Reference, Removed:Exec) {is=title K2Node_AsyncAction}
   > Set CooldownRemoveTask(CooldownRemoveTask:Ability Async Wait Gameplay Tag Removed Object Reference=@n8) -> (Output_Get:Ability Async Wait Gameplay Tag Removed Object Reference)
   Set Visibility(execute=@n8.Removed, Visibility=Collapsed, self=Get ProgressBar_130())
   > Clear and Invalidate Timer by Handle(Handle=Get CooldownHandle() -> (CooldownHandle:Timer Handle Structure))
   Set Visibility(execute=@n7.Added, Visibility=Visible, self=Get ProgressBar_130())
   > n9 = Set Timer by Event(Delegate=@n1, Time=0.1, bLooping=true)
   > Set CooldownHandle(CooldownHandle:Timer Handle Structure=@n9) -> (Output_Get:Timer Handle Structure)
   PrintString!(execute=@n6.CastFailed, InString=HeatSnake UI: Cannot Resolve Owner Pawn, bPrintToScreen, bPrintToLog, TextColor, Duration, Key) -> (then)
   Event Destruct()
   > Broadcast Event(self=GetGlobalEventRouter(), EventData=Make ST Heat Snake Auto Aim Event Data(ShouldShow_2_#1=False))
  ]}]
variables [
 As Fort Player Pawn Athena {type=FortPlayerPawnAthena}
 CooldownAddTask {type=AbilityAsync_WaitGameplayTagAdded, default=null}
 CooldownHandle {type=TimerHandle}
 CooldownListeningTag {type=GameplayTag, default {tagName=Granted.Cooldown.HeatSnake.Movement}}
 CooldownRemoveTask {type=AbilityAsync_WaitGameplayTagRemoved, default=null}
]
widgets {
 info {
  inheritedWidgetCount=1
  parentClass=/Script/FortniteUI.FortWeaponReticleExtensionWidgetBase
  rootWidgetClass=&Overlay
  widgetCount=5
 }
 widgets [
  Overlay_21 {widgetClassPath=&Overlay, widget=Overlay_21, properties {clipping=ClipToBounds, visibility=Hidden}}
  Image_21 {
   widgetClassPath=&Image
   bIsVariable=true
   parent=Overlay_21
   widget=Image_21
   properties {brush {
     bIsDynamicallyLoaded=false
     drawAs=Image
     imageType=FullColor
     mirroring=NoMirror
     resourceName=None
     resourceObject=/HeatSnake/UI/Materials/MI_Reticle_Centers_32px
     tiling=NoTile
     imageSize=<64 64>
     margin=<ltrb 0 0 0 0>
     outlineSettings {
      bUseBrushTransparency=false
      roundingType=HalfHeightRadius
      width=0
      color {colorUseRule=UseColor_Specified, specifiedColor=<rgba 0 0 0 0>}
      cornerRadii=<0 0 0 0>
     }
     tintColor {colorUseRule=UseColor_Specified, specifiedColor=<rgba 1 1 1 1>}
     uVRegion {bIsValid=false, max=<0 0>, min=<0 0>}
    }}
   slot {class=OverlaySlot}
  }
  ScaleBox_64 {
   widgetClassPath=&ScaleBox
   parent=Overlay_21
   widget=ScaleBox_64
   slot {class=OverlaySlot, properties {horizontalAlignment=HAlign_Fill, verticalAlignment=VAlign_Fill}}
  }
  ProgressBar_130 {
   widgetClassPath=/Script/UMG.ProgressBar
   bIsVariable=true
   parent=ScaleBox_64
   widget=ProgressBar_130
   properties {
    barFillType=FillFromCenterVertical
    percent=1
    widgetStyle {
     enableFillAnimation=false
     backgroundImage {
      bIsDynamicallyLoaded=false
      drawAs=Image
      imageType=FullColor
      mirroring=NoMirror
      resourceName=None
      resourceObject=&MI_Reticle_Centers_32px
      tiling=NoTile
      imageSize=<64 64>
      margin=<ltrb 0.4167 0.4167 0.4167 0.4167>
      outlineSettings {
       bUseBrushTransparency=false
       roundingType=HalfHeightRadius
       width=0
       color {colorUseRule=UseColor_Specified, specifiedColor=<rgba 0 0 0 0>}
       cornerRadii=<0 0 0 0>
      }
      tintColor {colorUseRule=UseColor_Specified, specifiedColor=<rgba 1 0.5 1 1>}
      uVRegion {bIsValid=false, max=<0 0>, min=<0 0>}
     }
     fillImage {
      bIsDynamicallyLoaded=false
      drawAs=Image
      imageType=FullColor
      mirroring=NoMirror
      resourceName=None
      resourceObject=&MI_Reticle_Centers_32px
      tiling=NoTile
      imageSize=<64 64>
      margin=<ltrb 0.4167 0.4167 0.4167 0.4167>
      outlineSettings {
       bUseBrushTransparency=false
       roundingType=HalfHeightRadius
       width=0
       color {colorUseRule=UseColor_Specified, specifiedColor=<rgba 0 0 0 0>}
       cornerRadii=<0 0 0 0>
      }
      tintColor {colorUseRule=UseColor_Specified, specifiedColor=<rgba 1 0 0 1>}
      uVRegion {bIsValid=false, max=<0 0>, min=<0 0>}
     }
     marqueeImage {
      bIsDynamicallyLoaded=false
      drawAs=Image
      imageType=FullColor
      mirroring=NoMirror
      resourceName=../../../Engine/Content/Slate/Common/ProgressBar_Marquee.png
      resourceObject=null
      tiling=Horizontal
      imageSize=<20 12>
      margin=<ltrb 0 0 0 0>
      outlineSettings {
       bUseBrushTransparency=false
       roundingType=HalfHeightRadius
       width=0
       color {colorUseRule=UseColor_Specified, specifiedColor=<rgba 0 0 0 0>}
       cornerRadii=<0 0 0 0>
      }
      tintColor {colorUseRule=UseColor_Specified, specifiedColor=<rgba 1 1 1 1>}
      uVRegion {bIsValid=false, max=<0 0>, min=<0 0>}
     }
    }
   }
   slot {class=ScaleBoxSlot, properties {horizontalAlignment=HAlign_Fill, verticalAlignment=VAlign_Fill}}
  }
  TutorialHighlight {widgetClassPath=&TutorialHighlight}
 ]
}
